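# Looping ambient sound cue for a console menu-select hum, classed as indoor ambience with an external attenuation asset and overridden per-owner concurrency.
# A mixer layers a pitched-down looping computer server bed with a quieter looping chain of randomly chosen computer one-shots.
# The one-shots get a random delay and pitch and volume variation on each pass.
# All quality levels share this mixer, whose other two inputs have no wave connected.
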SoundCue /Game/Sounds/Helios/Environment_Actors/ConsoleMenuSelectHum/Spot_Console_MenuSelectHum_Loop_Cue
root=QualityLevel_0
nodes [
 {id=Delay_0, children [Modulator_2], properties {delayMax=1, delayMin=0.5}}
 {id=Looping_0, children [Delay_0]}
 {id=Mixer_0, children [Modulator_1,Modulator_0,Looping_0,null], properties {inputVolume [0.6,0.5,0.1,0.05]}}
 {id=Modulator_0, children [WavePlayer_2], properties {pitchMax=0.8, pitchMin=0.8}}
 {id=Modulator_1, children [null], properties {pitchMax=0.8, pitchMin=0.8}}
 {id=Modulator_2, children [Random_1], properties {pitchMax=1.15, volumeMax=1, volumeMin=0.4}}
 {id=QualityLevel_0, children [Mixer_0,Mixer_0,Mixer_0,Mixer_0]}
 {id=Random_1, children [WavePlayer_5,WavePlayer_6,WavePlayer_7,WavePlayer_8], properties {weights [1,1,1,1]}}
 {
  id=WavePlayer_2
  properties {bLooping=true, soundWaveAssetPtr=/Game/Athena/Sounds/Quests/SiftingThroughTheArchives/Spot_Agency_Computer_Server_Loop}
 }
 {id=WavePlayer_5, properties {soundWaveAssetPtr=/Game/Sounds/Apollo/Biomes/GovernmentSecretComplex/SecretComplex_Computer_Oneshot_02}}
 {id=WavePlayer_6, properties {soundWaveAssetPtr=/Game/Sounds/Apollo/Biomes/GovernmentSecretComplex/SecretComplex_Computer_Oneshot_03}}
 {id=WavePlayer_7, properties {soundWaveAssetPtr=/Game/Sounds/Apollo/Biomes/GovernmentSecretComplex/SecretComplex_Computer_Oneshot_04}}
 {id=WavePlayer_8, properties {soundWaveAssetPtr=/Game/Sounds/Apollo/Biomes/GovernmentSecretComplex/SecretComplex_Computer_Oneshot_05}}
]
properties {
 attenuationSettings=&/Spot_Console_MenuSelectHum_Att
 bOverrideConcurrency=true
 maxDistance=550
 soundClassObject=&Fort_Ambient_Indoors_Athena
 subtitlePriority=5000
 volumeMultiplier=0.501
 attenuationOverrides {
  bEnableFocusInterpolation=true
  bEnableListenerFocus=true
  bEnableOcclusion=true
  dBAttenuationAtMax=-40
  falloffDistance=512
  focusAzimuth=15
  focusVolumeAttenuation=1.5
  nonFocusAzimuth=45
  nonFocusVolumeAttenuation=0.5
  occlusionInterpolationTime=1
  occlusionLowPassFilterFrequency=10000
  occlusionTraceChannel=ECC_GameTraceChannel11
  occlusionVolumeAttenuation=0.5
  stereoSpread=512
  attenuationShapeExtents=<256 64.5942 90>
 }
 concurrencyOverrides {bLimitToOwner=true, maxCount=2}
}
